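# Material function that overlays an animated, glitchy virus infection on the Crazy Eight character body's material attributes.
# A level animated from a Blueprint-set time, optionally quantized, moves and grows a distance mask between two pre-skinned positions.
# Inside it, noise and random square tiles in actor-anchored, depth-scaled screen space black out surface attributes.
# Emissive combines fresnel and cavity-map lines jittered by scrolling noise, with flicker and time-of-day scaling.
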MaterialFunction /BRCosmetics/Characters/Player/Male/Medium/Bodies/M_MED_Crazy_Eight/Materials/MF_Virus
graph {nodes [
  n1 = FunctionInput() {inputName=MaterialAttributes, inputType=FunctionInput_MaterialAttributes, sortPriority=-1}
  n2 = GetMaterialAttributes(None=@n1) {attributeGetTypes [
    #1
    #2
    #3
    #4
    #5
    #6
    #7
   ]}
  n3 = Constant3Vector()
  n4 = Time()
  n5 = Frac(None=Multiply(A=@n4) {constB=-2.25})
  n6 = Time()
  n7 = Round(None=Frac(None=Multiply(A=@n6) {constB=0.8}))
  n8 = Frac(None=Multiply(A=Add(A=@n6) {constB=0.33}) {constB=0.33})
  n9 = Multiply(A=Round(None=@n5), B=Multiply(A=@n7, B=Round(None=@n8)))
  n10 = MaterialFunctionCall(Float2=ViewSize()) {materialFunction=&BreakOutFloat2Components}
  n11 = AppendVector(A=Constant() {r=1}, B=Divide(A=@n10.G, B=@n10.G))
  n12 = Multiply(A=ConstantBiasScale(None=ScreenPosition().ViewportUV) {bias=-0.5, scale=2}, B=@n11)
  n13 = ComponentMask(None=PreSkinnedLocalBounds().Half Extents) {b=true}
  n14 = ScalarParameter() {group=Virus, parameterName=Virus_ScreenSpaceOffset}
  n15 = Add(
   A=ActorPositionWS()
   B=MaterialFunctionCall(Z=Add(A=@n13, B=@n14)) {materialFunction=&MakeFloat3}.Result
  ) {constB=100}
  n16 = Constant()
  n17 = Constant()
  n18 = Multiply(
   A=Multiply(
    A=Constant2Vector() {g=-0.5, r=0.5}
    B=MaterialFunctionCall(World Position=@n15) {materialFunction=&TransformToClipSpace}.Clip Space XY
   )
   B=If(A=Distance(A=@n15, B=@n16), B=Constant() {r=0.01}, "A > B"=Constant() {r=1}, "A == B"=@n17, "A < B"=@n17)
  )
  n19 = Constant() {r=100}
  n20 = ComponentMask(None=TransformPosition(None=@n15) {transformSourceType=TRANSFORMPOSSOURCE_World, transformType=TRANSFORMPOSSOURCE_View}) {b=true}
  n21 = MaterialFunctionCall(
   UVs=Subtract(A=ConstantBiasScale(None=@n12), B=Multiply(A=@n11, B=@n18))
   Texture Scale=Multiply(A=Divide(A=@n19, B=@n20), B=ScalarParameter() {defaultValue=0.01, group=Virus, parameterName=Virus_SquarePatternScale})
  ) {materialFunction=&ScaleUVsByCenter}
  n22 = MaterialFunctionCall(UVs=@n21.UVs, Texture Scale=Constant() {r=15}) {materialFunction=&ScaleUVsByCenter}
  n23 = ComponentMask(None=@n22.UVs) {r=true}
  n24 = ComponentMask(None=@n22.UVs) {g=true}
  n25 = Multiply(A=@n4) {constB=-0.01}
  n26 = AppendVector(A=Multiply(A=@n23) {constB=1.5}, B=Add(A=@n24, B=Frac(None=@n25)))
  n27 = AppendVector(A=Add(A=Multiply(A=@n9) {constB=0.05}, B=ComponentMask(None=@n26) {r=true}), B=ComponentMask(None=@n26) {g=true})
  n28 = TextureSample(UVs=ComponentMask(None=@n27) {g=true, r=true}) {samplerType=SAMPLERTYPE_LinearColor, texture=/Game/Global/Textures/Default/Noise/T_Noise01}
  n29 = ScalarParameter() {defaultValue=100, group=Virus, parameterName=Virus_CurrentLevel, sortPriority=3}
  n30 = Time()
  n31 = ScalarParameter() {defaultValue=-99999, group=Virus, parameterName=Virus_GameTimeFromBP, sortPriority=100}
  n32 = ScalarParameter() {defaultValue=2, group=Virus, parameterName=Virus_Animation_Duration, sortPriority=101}
  n33 = Saturate(None=Divide(A=Subtract(A=@n30, B=@n31), B=Add(A=@n32) {constB=0}))
  n34 = ScalarParameter() {defaultValue=500, group=Virus, parameterName=Virus_StopAtLevel, sortPriority=3}
  n35 = ScalarParameter() {defaultValue=100, group=Virus, parameterName=Virus_StartAtLevel, sortPriority=3}
  n36 = Subtract(A=@n34, B=@n35)
  n37 = Subtract(A=Power(Base=@n33, Exp=ScalarParameter() {defaultValue=1.33, group=Virus, parameterName=Virus_Animation_EaseIn, sortPriority=103}) {constExponent=1.33}) {constB=0.5}
  n38 = OneMinus(None=Abs(None=Multiply(A=@n37) {constB=2}))
  n39 = LinearInterpolate(
   A=LinearInterpolate(A=Subtract(A=@n29), B=@n29, Alpha=@n33)
   B=Add(A=@n29, B=@n36)
   Alpha=Power(Base=@n38, Exp=ScalarParameter() {defaultValue=6, group=Virus, parameterName=Virus_Animation_EaseInOut, sortPriority=103}) {constExponent=8}
  )
  n40 = Divide(A=Subtract(A=Clamp(None=@n39, Min=@n35, Max=@n34), B=@n35), B=@n36)
  n41 = Max(A=Ceil(None=ScalarParameter() {defaultValue=1, group=Virus, parameterName=Virus_QuantizationSteps, sliderMax=100, sliderMin=1}))
  n42 = Multiply(A=@n40, B=@n41)
  n43 = Subtract(A=@n41)
  n44 = LinearInterpolate(A=@n40, B=Divide(A=Floor(None=@n42), B=@n41), Alpha=Saturate(None=Ceil(None=@n43)))
  n45 = Distance(
   A=Add(
    A=PreSkinnedPosition()
    B=VectorParameter() {group=Virus, parameterName=Virus_Backpack_PreskinnedPositionOffset, defaultValue=<rgba 0 0 0 1>}.RGB
   )
   B=LinearInterpolate(
    A=VectorParameter() {group=Virus, parameterName=Virus_StartingPosition, defaultValue=<rgba -10 0.5 156 1>}.RGB
    B=VectorParameter() {group=Virus, parameterName=Virus_EndingPosition, defaultValue=<rgba -96 0.5 105 1>}.RGB
    Alpha=Multiply(A=@n44, B=@n44)
   )
  )
  n46 = ScalarParameter() {defaultValue=20, group=Virus, parameterName=Virus_BlendOutWidth, sortPriority=3}
  n47 = LinearInterpolate(
   A=Subtract(A=ScalarParameter() {defaultValue=-14, group=Virus, parameterName=Virus_StartingCoverDistance, sortPriority=1}, B=@n46)
   B=ScalarParameter() {defaultValue=165, group=Virus, parameterName=Virus_FullCoverDistance, sortPriority=2}
   Alpha=@n44
  )
  n48 = ComponentMask(None=VertexInterpolator(VS=Subtract(A=@n45, B=@n47) {constB=160}).PS) {r=true}
  n49 = Subtract(A=@n48) {constB=-10}
  n50 = Divide(B=Add(A=@n46) {constB=0})
  n51 = ScalarParameter() {defaultValue=2, group=Virus, parameterName=Virus_BlendOutPower, sortPriority=4}
  n52 = Power(Base=Saturate(None=Multiply(A=@n49, B=@n50) {constB=0.05}), Exp=@n51)
  n53 = Multiply(A=Power(Base=Multiply(A=@n28.R)) {constExponent=0.5}, B=@n52)
  n54 = Saturate(None=Multiply(A=@n48, B=@n50) {constB=0.05})
  n55 = OneMinus(None=Power(Base=@n54, Exp=@n51))
  n56 = Multiply(A=OneMinus(None=@n53), B=@n55)
  n57 = VertexInterpolator(VS=@n21.UVs)
  n58 = Subtract(A=Frac(None=@n57.PS)) {constB=0.5}
  n59 = Saturate(None=Abs(None=Multiply(A=@n58) {constB=2}))
  n60 = Subtract(A=@n57.PS) {constB=0.5}
  n61 = MaterialFunctionCall!(
   UVs_Layer1=@n60
   UVs_Layer2=@n60
   Seed_Layer1=Add(A=Constant() {r=2.651}, B=@n7)
   Seed_Layer2=Add(A=@n8, B=Constant() {r=5.1561})
   Threshold_Layer1=Constant() {r=0.5216}
   Threshold_Layer2=Constant() {r=0.4871}
   EdgeMaskWidth_Layer1
   EdgeMaskWidth_Layer2
  ) {materialFunction=/Game/Weapons/WeaponSkins/MaterialFunction/MF_RandomTilePatternFromUVs}
  n62 = Add(A=@n59, B=Multiply(A=Constant() {r=0.75}, B=Multiply(A=@n61.Layer1_RandomValue, B=@n55))) {constB=0.25}
  n63 = ComponentMask(None=@n62) {r=true}
  n64 = ComponentMask(None=@n62) {g=true}
  n65 = Multiply(A=@n61.Layer1_Mask, B=Multiply(A=Round(None=@n63), B=Round(None=@n64)))
  n66 = Add(A=@n59, B=Multiply(A=Constant() {r=-0.75}, B=Multiply(A=@n61.Layer1_RandomValue, B=@n55))) {constB=0.25}
  n67 = ComponentMask(None=@n66) {r=true}
  n68 = ComponentMask(None=@n66) {g=true}
  n69 = Multiply(A=OneMinus(None=@n61.Layer1_Mask), B=Multiply(A=Round(None=@n67), B=Round(None=@n68)))
  n70 = Multiply(A=Multiply(A=@n61.Layer2_Mask, B=Add(A=@n65, B=@n69)), B=OneMinus(None=Power(Base=@n54) {constExponent=5}))
  n71 = OneMinus(None=Power(Base=@n70) {constExponent=0.5})
  n72 = Constant() {r=1}
  n73 = Multiply(A=@n70, B=@n54)
  n74 = Power(Base=@n38) {constExponent=3}
  n75 = MaterialFunctionCall(Power=ScalarParameter() {defaultValue=12, group=Virus, parameterName=Virus_FresnelPower}) {materialFunction=&Fresnel_Function}
  n76 = AppendVector(A=@n23, B=Add(A=@n24, B=@n5))
  n77 = TextureSample(UVs=ComponentMask(None=@n76) {g=true, r=true}) {samplerType=SAMPLERTYPE_Grayscale, texture=&LowResBlurredNoise}
  n78 = ComponentMask(None=@n9) {r=true}
  n79 = Multiply(A=Saturate(None=@n78)) {constB=0.33}
  n80 = Max(A=@n52, B=Add(A=@n79, B=Constant() {r=0.5}) {constB=0.5}) {constB=0.5}
  n81 = Multiply(A=Multiply(A=Subtract(A=@n77.G) {constB=0.25}, B=@n80)) {constB=0.025}
  n82 = TextureCoordinate()
  n83 = TextureObjectParameter() {group=Virus, parameterName=Virus_Cavity, texture=&T_BlacksRGB}
  n84 = TextureSample(UVs=Add(A=Multiply(A=@n81) {constB=-0.5}, B=@n82), Tex=@n83) {texture=/BRCosmetics/Characters/Player/Male/Medium/Bodies/M_MED_Crazy_Eight/Textures/Crazy_Eight_Body_Cavity}
  n85 = Multiply(A=@n78) {constB=0.33}
  n86 = Frac(None=Multiply(A=Time()) {constB=0.4})
  n87 = Add(A=Multiply(A=Sine(None=@n86)) {constB=0.2})
  n88 = Add(A=Multiply(A=Multiply(A=@n74, B=@n73)) {constB=0.2}, B=Multiply(A=Add(A=@n75.Result, B=@n84.G), B=Add(A=@n85, B=@n87)))
  n89 = VectorParameter() {group=Virus, parameterName=Virus_LineEmissiveColor, defaultValue=<rgba 1 0.0737 0.025 1>}
  n90 = Add(A=@n74)
  n91 = Add(A=@n82, B=Multiply(A=@n81, B=@n90))
  n92 = Multiply(A=@n78) {constB=0.0033}
  n93 = VectorParameter() {group=Virus, parameterName=Virus_LineEmissiveColor3, defaultValue=<rgba 8 2 2 1>}
  n94 = Add(A=Multiply(A=@n79) {constB=0.33}, B=@n74)
  n95 = Multiply(
   A=VectorParameter() {group=Virus, parameterName=Virus_LineEmissiveColor2, defaultValue=<rgba 0.5 0.0844 0.0625 1>}.RGB
   B=Add(A=Multiply(A=@n93.RGB, B=@n94))
  )
  n96 = Multiply(
   A=TextureSample(UVs=Add(A=@n91, B=@n92), Tex=@n83) {texture=&Crazy_Eight_Body_Cavity}.G
   B=@n95
  )
  n97 = Multiply(A=@n4) {constB=-1.5}
  n98 = AppendVector(A=Add(A=Frac(None=@n97), B=@n23), B=@n24)
  n99 = Subtract(A=TextureSample(UVs=ComponentMask(None=@n98) {g=true, r=true}) {samplerType=SAMPLERTYPE_Grayscale, texture=&LowResBlurredNoise}.G) {constB=0.25}
  n100 = Multiply(A=@n90, B=Multiply(A=Multiply(A=@n99, B=@n80)) {constB=0.025})
  n101 = TextureSample(UVs=Subtract(A=Add(A=@n82, B=@n100), B=@n92), Tex=@n83) {texture=&Crazy_Eight_Body_Cavity}
  n102 = Add(A=Add(A=Multiply(A=@n88, B=@n89.RGB), B=Multiply(A=@n96, B=@n80)), B=Multiply(A=Multiply(A=@n95, B=@n101.G), B=@n80))
  n103 = Constant() {r=0.9}
  n104 = MaterialFunctionCall(Value_Morning=@n103, Value_Day=Constant() {r=1}, Value_Evening=@n103, Value_Night=Constant() {r=0.5}) {materialFunction=&Scalar_Time_of_Day}
  n105 = Subtract(A=@n48) {constB=15}
  n106 = Power(Base=Saturate(None=Multiply(A=@n105, B=@n50) {constB=0.05}), Exp=@n51)
  n107 = SetMaterialAttributes!(
   MaterialAttributes=@n1
   BaseColor=Multiply(A=LinearInterpolate(A=@n2.BaseColor, B=@n3, Alpha=@n56), B=@n71)
   EmissiveColor=LinearInterpolate(A=@n2.EmissiveColor, B=Multiply(A=@n102, B=@n104.Result), Alpha=OneMinus(None=@n106))
   Metallic=Multiply(A=LinearInterpolate(A=@n2.Metallic, B=@n3, Alpha=@n56), B=@n71)
   Specular=Multiply(A=LinearInterpolate(A=@n2.Specular, B=@n3, Alpha=@n56), B=@n71)
   Roughness=Multiply(A=LinearInterpolate(A=@n2.Roughness, B=@n3, Alpha=@n56), B=@n71)
   CustomData1=Multiply(A=LinearInterpolate(A=@n2.CustomData1, B=@n72, Alpha=@n56), B=@n71)
   CustomData0=Multiply(A=LinearInterpolate(A=@n2.CustomData0, B=@n3, Alpha=@n56), B=@n71)
   ShadingModel=ShadingModel()
  ) {attributeSetTypes [
    #1
    #2
    #3
    #4
    #5
    #6
    #7
    #8
   ]}
  FunctionOutput(None=@n107)
  ObjectRadius()
  Power(Base=@n53) {constExponent=0.25}
  VertexInterpolator(VS=@n9)
  VertexInterpolator(VS=@n22.UVs)
  VertexInterpolator(VS=@n27)
  VertexInterpolator(VS=@n76)
  VertexInterpolator(VS=@n98)
 ]}
