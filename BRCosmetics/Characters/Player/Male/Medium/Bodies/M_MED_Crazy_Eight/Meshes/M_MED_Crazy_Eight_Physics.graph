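# Physics asset for the M_MED_Crazy_Eight male player body cosmetic, covering the torso and a left-side dangling knife.
# Capsules on the pelvis and spine form an angularly limited chain.
# Tiny MSTR boxes act as anchors with free rotation: the knife's on spine_05, the pocket's on pelvis.
# The knife box hangs from its anchor under tight swing and twist limits, while the pocket anchor has no child body.
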PhysicsAsset /BRCosmetics/Characters/Player/Male/Medium/Bodies/M_MED_Crazy_Eight/Meshes/M_MED_Crazy_Eight_Physics
bodies {
 dyn_knife_01_l [{shapeType=Box, extentX=7.0828, extentY=3.4176, extentZ=32.8313, center=<-0.6348 0.3382 0.824>, rotation=<rot 0 -90.9273 -88.6503>}]
 dyn_knife_01_l_MSTR [{shapeType=Box, extentX=0.01, extentY=0.01, extentZ=0.01}]
 dyn_pocket_01_l_MSTR [{shapeType=Box, extentX=0.01, extentY=0.01, extentZ=0.01}]
 pelvis [{shapeType=Capsule, length=18.8264, radius=8.22}]
 spine_01 [{shapeType=Capsule, length=12.9029, radius=10.7891}]
 spine_03 [{shapeType=Capsule, length=11.9838, radius=9.5395}]
 spine_04 [{shapeType=Capsule, length=6.6208, radius=16.3944}]
 spine_05 [{shapeType=Capsule, length=21.6586, radius=12.3366, center=<0 -1.3701 0>}]
}
constraints [
 {bone1Name=spine_01, bone2Name=pelvis, swing1LimitDegrees=45, swing2LimitDegrees=45, twistLimitDegrees=45}
 {bone1Name=spine_03, bone2Name=spine_01, swing1LimitDegrees=45, swing2LimitDegrees=45, twistLimitDegrees=45}
 {bone1Name=spine_04, bone2Name=spine_03, swing1LimitDegrees=45, swing2LimitDegrees=45, twistLimitDegrees=45}
 {bone1Name=spine_05, bone2Name=spine_04, swing1LimitDegrees=45, swing2LimitDegrees=45, twistLimitDegrees=45}
 {bone1Name=dyn_knife_01_l, bone2Name=dyn_knife_01_l_MSTR, swing1LimitDegrees=10, swing2LimitDegrees=10, twistLimitDegrees=5}
 {
  bone1Name=dyn_knife_01_l_MSTR
  bone2Name=spine_05
  swing1LimitDegrees=45
  swing1Motion=Free
  swing2LimitDegrees=45
  swing2Motion=Free
  twistLimitDegrees=45
  twistMotion=Free
 }
 {
  bone1Name=dyn_pocket_01_l_MSTR
  bone2Name=pelvis
  swing1LimitDegrees=45
  swing1Motion=Free
  swing2LimitDegrees=45
  swing2Motion=Free
  twistLimitDegrees=45
  twistMotion=Free
 }
]
