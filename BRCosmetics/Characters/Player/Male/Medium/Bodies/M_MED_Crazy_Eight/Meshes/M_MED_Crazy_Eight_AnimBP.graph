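# Animation Blueprint for the Crazy Eight male medium cosmetic body, layering secondary motion onto the parent mesh's copied pose.
# A weighted multi-bone constraint drives the left knife, two RigidBody nodes simulate physics, and AnimDynamics springs the gloves.
# Glute rotation constraints and additive scale on glutes and upper-arm twists act as correctives, weighted by limb yaw from ref pose, with skydiving and crouch overrides.
# Aiming alphas are computed but unused in the AnimGraph.
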AnimBlueprint /BRCosmetics/Characters/Player/Male/Medium/Bodies/M_MED_Crazy_Eight/Meshes/M_MED_Crazy_Eight_AnimBP
target_skeleton=&/M_MED_Crazy_Eight_Skeleton
graphs [
 AnimGraph {nodes [
   n1 = "Constraint - dyn_knife_01_l_MSTR (PPPPPPPPPP)"!(
    ComponentPose=Local To Component(LocalPose=Copy Pose From Mesh() {bCopyCurves=true, bUseAttachedParent=true})
    Alpha=1
    ConstraintWeights_0=0.369
    ConstraintWeights_1:float=0.0041
    ConstraintWeights_2:float=0.0037
    ConstraintWeights_3:float=0.0397
    ConstraintWeights_4:float=0.1021
    ConstraintWeights_5:float=0.246
    ConstraintWeights_6:float=0.2216
    ConstraintWeights_7:float=0.0051
    ConstraintWeights_8:float=0.0074
    ConstraintWeights_9:float=0.0014
   ) -> (Pose) {
    alphaBoolBlend {blendOption=Linear}
    boneToModify {boneName=dyn_knife_01_l_MSTR}
    constraintSetup [
     {targetBone {boneName=clavicle_l}}
     {targetBone {boneName=upperarm_twist_02_l}}
     {targetBone {boneName=spine_01}}
     {targetBone {boneName=spine_02}}
     {targetBone {boneName=spine_03}}
     {targetBone {boneName=spine_04}}
     {targetBone {boneName=spine_05}}
     {targetBone {boneName=clavicle_r}}
     {targetBone {boneName=neck_02}}
     {targetBone {boneName=neck_01}}
    ]
    constraintWeights [1,1,1,1,1,1,1,1,1,1]
   }
   n2 = RigidBody(ComponentPose=@n1) {
    bClampLinearTranslationLimitToRefPose=true
    bOverrideWorldGravity=true
    overridePhysicsAsset=&/M_MED_Crazy_Eight_Physics
    simulationSpace=BaseBoneSpace
    alphaBoolBlend {blendOption=Linear}
    baseBoneRef {boneName=spine_01}
    componentAppliedLinearAccClamp=<3000 3000 3000>
    componentLinearAccScale=<0.3 0.3 0.3>
    componentLinearVelScale=<0.3 0.3 0.3>
    overrideWorldGravity {z=-980}
   }
   n3 = RigidBody(ComponentPose=@n2) {
    bClampLinearTranslationLimitToRefPose=true
    bOverrideWorldGravity=true
    overridePhysicsAsset=&/M_MED_Crazy_Eight_L_leg_Physics
    simulationSpace=BaseBoneSpace
    alphaBoolBlend {blendOption=Linear}
    baseBoneRef {boneName=thigh_l}
    componentAppliedLinearAccClamp=<3000 3000 3000>
    componentLinearAccScale=<0.3 0.3 0.3>
    componentLinearVelScale=<0.3 0.3 0.3>
    overrideWorldGravity {z=-980}
   }
   n4 = Anim Dynamics - Bone: dyn_glove_01_l(ComponentPose=@n3, Alpha=1, ExternalForce="0,0,0") {
    angularDampingOverride=1
    angularSpringConstant=150
    bAngularSpring=true
    bGravityOverrideInSimSpace=true
    simulationSpace=BoneRelative
    alphaBoolBlend {blendOption=Linear}
    boundBone {boneName=dyn_glove_01_l}
    physicsBodyDefinitions [{
      sphereCollisionRadius=0
      boundBone {boneName=dyn_glove_01_l}
      boxExtents=<10 10 10>
      constraintSetup {angularLimitsMax=<10 10 10>, angularLimitsMin=<-10 -10 -10>, angularTarget=<0 0 0>}
      localJointOffset=<0 0 0.1>
     }]
    relativeSpaceBone {boneName=hand_l}
   }
   n5 = Anim Dynamics - Bone: dyn_glove_01_r(ComponentPose=@n4, Alpha=1, ExternalForce="0,0,0") {
    angularDampingOverride=1
    angularSpringConstant=150
    bAngularSpring=true
    bGravityOverrideInSimSpace=true
    simulationSpace=BoneRelative
    alphaBoolBlend {blendOption=Linear}
    boundBone {boneName=dyn_glove_01_r}
    physicsBodyDefinitions [{
      sphereCollisionRadius=0
      boundBone {boneName=dyn_glove_01_r}
      boxExtents=<10 10 10>
      constraintSetup {angularLimitsMax=<10 10 10>, angularLimitsMin=<-10 -10 -10>, angularTarget=<0 0 0>}
      localJointOffset=<0 0 0.1>
     }]
    relativeSpaceBone {boneName=hand_r}
   }
   n6 = "Constraint - glut_01_l (RR)"!(ComponentPose=@n5, Alpha=1, ConstraintWeights_0=0.2, ConstraintWeights_1:float=0.8) -> (Pose) {
    alphaBoolBlend {blendOption=Linear}
    boneToModify {boneName=glut_01_l}
    constraintSetup [{transformType=Rotation, targetBone {boneName=thigh_l}}, {transformType=Rotation, targetBone {boneName=pelvis}}]
    constraintWeights [1,1]
   }
   n7 = "Constraint - glut_01_r (RR)"!(ComponentPose=@n6, Alpha=1, ConstraintWeights_0=0.2, ConstraintWeights_1:float=0.8) -> (Pose) {
    alphaBoolBlend {blendOption=Linear}
    boneToModify {boneName=glut_01_r}
    constraintSetup [{transformType=Rotation, targetBone {boneName=thigh_r}}, {transformType=Rotation, targetBone {boneName=pelvis}}]
    constraintWeights [1,1]
   }
   n8 = "Transform (Modify) Bone - Bone: glut_01_l"(ComponentPose=@n7, Alpha=Get thighSwingDelta_l() -> (thighSwingDelta_l:double), Translation="0,0,0", Rotation="0,0,0", Scale="1.25,1,1") {scaleMode=BMM_Additive, scaleSpace=BCS_BoneSpace, alphaBoolBlend {blendOption=Linear}, boneToModify {boneName=glut_01_l}, scale {x=1.25}}
   n9 = "Transform (Modify) Bone - Bone: glut_01_r"(ComponentPose=@n8, Alpha=Get thighSwingDelta_r() -> (thighSwingDelta_r:double), Translation="0,0,0", Rotation="0,0,0", Scale="1.25,1,1") {scaleMode=BMM_Additive, scaleSpace=BCS_BoneSpace, alphaBoolBlend {blendOption=Linear}, boneToModify {boneName=glut_01_r}, scale {x=1.25}}
   n10 = "Transform (Modify) Bone - Bone: upperarm_twist_01_r"(ComponentPose=@n9, Alpha=Get armSwingDelta_r() -> (armSwingDelta_r:double), Translation="0,0,0", Rotation="0,0,0", Scale="1.5,1.5,1.5") {scaleMode=BMM_Additive, alphaBoolBlend {blendOption=Linear}, boneToModify {boneName=upperarm_twist_01_r}, scale=<1.5 1.5 1.5>}
   n11 = "Transform (Modify) Bone - Bone: upperarm_twist_01_l"(ComponentPose=@n10, Alpha=Get armSwingDelta_l() -> (armSwingDelta_l:double), Translation="0,0,0", Rotation="0,0,0", Scale="1.5,1.5,1.5") {scaleMode=BMM_Additive, alphaBoolBlend {blendOption=Linear}, boneToModify {boneName=upperarm_twist_01_l}, scale=<1.5 1.5 1.5>}
   Output Pose(Result=Component To Local(ComponentPose=@n11)) {layerGroup=None}
  ]}
 EventGraph {nodes [
   Event BlueprintUpdateAnimation()
   > n1 = Sequence()
   n2 = Get bIsTargeting()
   > Set isAimingClampedAlpha(isAimingClampedAlpha:double="To Float (Boolean)"(InBool=@n2)) -> (Output_Get:double)
   > Set isNotAimingAlpha(isNotAimingAlpha:double=SelectFloat(A=1, B=0.5, bPickA=NOT Boolean(A=@n2))) -> (Output_Get:double)
   n3 = GetOwnerSkeletalMeshComponent()
   n4 = GetDeltaTransformFromRefPose(self=@n3, BoneName=upperarm_r, BaseName=clavicle_r)
   n5 = "Absolute (Float)"(A=BreakRotator(InRot=BreakTransform(InTransform=@n4).Rotation).Yaw)
   n6 = Get bIsSkyDiving()
   Set armSwingDelta_r(
    execute=@n1.then_1
    armSwingDelta_r:double=SelectFloat(A=0.5, B=MapRangeClamped(Value=@n5, InRangeA=90, InRangeB=130, OutRangeB=0.75), bPickA=@n6)
   ) -> (Output_Get:double)
   n7 = BreakRotator(InRot=BreakTransform(InTransform=GetDeltaTransformFromRefPose(self=@n3, BoneName=upperarm_l, BaseName=clavicle_l)).Rotation)
   > Set armSwingDelta_l(armSwingDelta_l:double=SelectFloat(A=0.5, B=MapRangeClamped(Value=@n7.Yaw, InRangeA=10, InRangeB=80), bPickA=@n6)) -> (Output_Get:double)
   n8 = GetDeltaTransformFromRefPose(self=@n3, BoneName=thigh_l, BaseName=pelvis)
   n9 = "Absolute (Float)"(A=BreakRotator(InRot=BreakTransform(InTransform=@n8).Rotation).Yaw)
   > Set thighSwingDelta_l(thighSwingDelta_l:double=SelectFloat(A=1, B=MapRangeClamped(Value=@n9, InRangeA=50, InRangeB=110), bPickA=Get bIsCrouching())) -> (Output_Get:double)
   n10 = BreakRotator(InRot=BreakTransform(InTransform=GetDeltaTransformFromRefPose(self=@n3, BoneName=thigh_r, BaseName=pelvis)).Rotation)
   > Set thighSwingDelta_r(thighSwingDelta_r:double=MapRangeClamped(Value="Absolute (Float)"(A=@n10.Yaw), InRangeA=50, InRangeB=110)) -> (Output_Get:double)
   "Absolute (Float)"(A=@n7.Yaw)
  ]}
]
variables [
 armSwingDelta_l {type=number}
 armSwingDelta_r {type=number}
 isAimingClampedAlpha {type=number}
 isNotAimingAlpha {type=number}
 thighSwingDelta_l {type=number}
 thighSwingDelta_r {type=number}
]
